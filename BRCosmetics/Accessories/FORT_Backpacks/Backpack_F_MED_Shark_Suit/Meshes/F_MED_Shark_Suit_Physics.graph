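# Physics asset for the Shark Suit back bling mesh, giving its dynamic fin bones limited physical movement.
# Capsule bodies cover the root, pack, side fins, top fin and a four-bone tail fin.
# The pack has all angular motion locked to the root and anchors the side fins, top fin and tail chain.
# Fin joints use tight, mostly asymmetric swing limits with little or no twist, keeping sway narrow and directional.
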PhysicsAsset /BRCosmetics/Accessories/FORT_Backpacks/Backpack_F_MED_Shark_Suit/Meshes/F_MED_Shark_Suit_Physics
bodies {
 dyn_side_fin_l_01 [{shapeType=Capsule, length=5.25, radius=3.75, shapeName=dyn_side_fin_l_01_capsule, center=<2.5132 0 0>, rotation=<rot 80 0 0>}]
 dyn_side_fin_r_01 [{shapeType=Capsule, length=4, radius=4, shapeName=dyn_side_fin_r_01_capsule, center=<2.6816 0 -0.4728>, rotation=<rot 80 180 180>}]
 dyn_tail_fin_01 [{shapeType=Capsule, length=1.8083, radius=3.895, center=<3.9875 0 0>, rotation=<rot 90 0 0>}]
 dyn_tail_fin_02 [{shapeType=Capsule, length=2.5476, radius=3.1556, center=<3.9866 0.0833 0>, rotation=<rot 90 0 0>}]
 dyn_tail_fin_03 [{shapeType=Capsule, length=5.1198, radius=2.3909, center=<5.2877 0.0509 0>, rotation=<rot 90 0 0>}]
 dyn_tail_fin_04 [{shapeType=Capsule, length=3.7033, radius=2, center=<0 -0.0176 -1.8193>, rotation=<rot 90 0 0>}]
 dyn_top_fin_01 [{shapeType=Capsule, length=5.75, radius=3.25, shapeName=dyn_top_fin_01_capsule, center=<4.9432 0 0>, rotation=<rot -90 14.0362 5.9638>}]
 pack [{shapeType=Capsule, length=2, radius=7.0142, center=<0 -5.4291 0>}]
 root [{shapeType=Capsule, length=40.25, radius=15, shapeName=root_capsule, center=<0 15.6413 0>}]
}
constraints [
 {bone1Name=dyn_tail_fin_04, bone2Name=dyn_tail_fin_03, swing1LimitDegrees=12, swing2LimitDegrees=7, twistLimitDegrees=0}
 {bone1Name=dyn_tail_fin_03, bone2Name=dyn_tail_fin_02, swing1LimitDegrees=5, swing2LimitDegrees=12, twistLimitDegrees=0}
 {bone1Name=dyn_tail_fin_02, bone2Name=dyn_tail_fin_01, swing1LimitDegrees=5, swing2LimitDegrees=12, twistLimitDegrees=0}
 {bone1Name=dyn_tail_fin_01, bone2Name=pack, swing1LimitDegrees=5, swing2LimitDegrees=12, twistLimitDegrees=0}
 {
  bone1Name=pack
  bone2Name=root
  swing1LimitDegrees=45
  swing1Motion=Locked
  swing2LimitDegrees=45
  swing2Motion=Locked
  twistLimitDegrees=45
  twistMotion=Locked
 }
 {bone1Name=dyn_top_fin_01, bone2Name=pack, swing1LimitDegrees=1, swing2LimitDegrees=20, twistLimitDegrees=1}
 {bone1Name=dyn_side_fin_r_01, bone2Name=pack, swing1LimitDegrees=20, swing2LimitDegrees=2, twistLimitDegrees=1.5}
 {bone1Name=dyn_side_fin_l_01, bone2Name=pack, swing1LimitDegrees=20, swing2LimitDegrees=2, twistLimitDegrees=1.5}
]
