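# Physics asset for the shark suit backpack cosmetic, named as the airborne variant, that simulates the fins as jointed bodies.
# The pack capsule is angularly locked to a root box and anchors the top fin, both side fins and a four-segment capsule tail chain.
# Tail joints allow moderate swing with little twist, tightening at the tip, while each box-shaped fin mainly swings on one axis.
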PhysicsAsset /BRCosmetics/Accessories/FORT_Backpacks/Backpack_F_MED_Shark_Suit/Meshes/F_MED_Shark_Suit_Physics_Airborne
bodies {
 dyn_side_fin_l_01 [{shapeType=Box, extentX=8.1971, extentY=2, extentZ=8.255, center=<4.2383 0 0>}]
 dyn_side_fin_r_01 [{shapeType=Box, extentX=8.4585, extentY=2, extentZ=7.8101, center=<4.4794 0 0>}]
 dyn_tail_fin_01 [{shapeType=Capsule, length=1.8083, radius=3.895, center=<3.9875 0 0>, rotation=<rot 90 0 0>}]
 dyn_tail_fin_02 [{shapeType=Capsule, length=2.5476, radius=3.1556, center=<3.9866 0.0833 0>, rotation=<rot 90 0 0>}]
 dyn_tail_fin_03 [{shapeType=Capsule, length=5.1198, radius=2.3909, center=<5.2877 0.0509 0>, rotation=<rot 90 0 0>}]
 dyn_tail_fin_04 [{shapeType=Capsule, length=3.7033, radius=2, center=<0 -0.0176 -1.8193>, rotation=<rot 90 0 0>}]
 dyn_top_fin_01 [{shapeType=Box, extentX=10.1352, extentY=8.4495, extentZ=2, center=<5.4365 0 0>}]
 pack [{shapeType=Capsule, length=2, radius=7.0142, center=<0 -5.4291 0>}]
 root [{shapeType=Box, extentX=3.0048, extentY=36.9197, extentZ=62.751, center=<0 2.7064 -6.7521>, rotation=<rot 0 90 0>}]
}
constraints [
 {bone1Name=dyn_tail_fin_04, bone2Name=dyn_tail_fin_03, swing1LimitDegrees=25, swing2LimitDegrees=15, twistLimitDegrees=3}
 {bone1Name=dyn_tail_fin_03, bone2Name=dyn_tail_fin_02, swing1LimitDegrees=40, swing2LimitDegrees=25, twistLimitDegrees=3}
 {bone1Name=dyn_tail_fin_02, bone2Name=dyn_tail_fin_01, swing1LimitDegrees=40, swing2LimitDegrees=25, twistLimitDegrees=3}
 {bone1Name=dyn_tail_fin_01, bone2Name=pack, swing1LimitDegrees=40, swing2LimitDegrees=25, twistLimitDegrees=3}
 {
  bone1Name=pack
  bone2Name=root
  swing1LimitDegrees=45
  swing1Motion=Locked
  swing2LimitDegrees=45
  swing2Motion=Locked
  twistLimitDegrees=45
  twistMotion=Locked
 }
 {bone1Name=dyn_top_fin_01, bone2Name=pack, swing1LimitDegrees=1, swing2LimitDegrees=40, twistLimitDegrees=1}
 {bone1Name=dyn_side_fin_l_01, bone2Name=pack, swing1LimitDegrees=20, swing2LimitDegrees=5, twistLimitDegrees=1}
 {bone1Name=dyn_side_fin_r_01, bone2Name=pack, swing1LimitDegrees=20, swing2LimitDegrees=5, twistLimitDegrees=1}
]
